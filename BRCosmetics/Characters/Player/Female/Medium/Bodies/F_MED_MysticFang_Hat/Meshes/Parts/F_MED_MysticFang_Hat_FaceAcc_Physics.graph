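# Physics asset for the face accessory part of the female medium MysticFang Hat body, covering hair and hat bones.
# A back-hair chain with swing limits and locked twist, plus a short, tightly limited front-hair chain, hang from the head.
# Capsules on torso, neck, head and upper arms use free angular constraints, some skipping spine links.
# Hat-back and left-forehead capsules are rotation-locked to the head; a stray unnamed body duplicates the spine_04 capsule.
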PhysicsAsset /BRCosmetics/Characters/Player/Female/Medium/Bodies/F_MED_MysticFang_Hat/Meshes/Parts/F_MED_MysticFang_Hat_FaceAcc_Physics
bodies {
 FACIAL_L_ForeheadIn [
  {shapeType=Capsule, length=5, radius=2, shapeName=FACIAL_L_ForeheadIn_capsule, center=<-3.1121 0 1.0803>, rotation=<rot 0 90 90>}
  {shapeType=Capsule, length=5, radius=3, shapeName=FACIAL_L_ForeheadIn_capsule, center=<-2.3326 2.2725 -0.1315>, rotation=<rot 0 0 -82.0067>}
 ]
 FACIAL_L_ForeheadMid [
  {shapeType=Capsule, length=5, radius=2, shapeName=FACIAL_L_ForeheadMid_capsule, center=<0 1.0111 1.7503>, rotation=<rot 0 90 90>}
  {shapeType=Capsule, length=5, radius=3, shapeName=FACIAL_L_ForeheadMid_capsule, center=<-1.5443 3.5029 0.1583>, rotation=<rot 0 0 102.6753>}
 ]
 None [{shapeType=Capsule, length=15, radius=10, shapeName=spine_04_capsule}]
 dyn_Hair_Back_02 [{shapeType=Capsule, length=19.396, radius=4, shapeName=dyn_Hair_Back_02_capsule, center=<2.5 0 2>, rotation=<rot 0 0 90>}]
 dyn_Hair_Back_03 [{shapeType=Capsule, length=10, radius=6, shapeName=dyn_Hair_Back_03_capsule, center=<8 0 3>, rotation=<rot 0 0 90>}]
 dyn_Hair_Back_04 [{shapeType=Capsule, length=10, radius=6, shapeName=dyn_Hair_Back_04_capsule, center=<10 0 0>, rotation=<rot 0 0 90>}]
 dyn_Hair_Back_06 [{shapeType=Capsule, length=10, radius=6, shapeName=dyn_Hair_Back_06_capsule, center=<9.2655 -2 2>, rotation=<rot 0 0 90>}]
 dyn_hair_front_a_01_m [{shapeType=Capsule, length=3, radius=1, shapeName=dyn_hair_front_a_01_m_capsule, center=<3 3 0>, rotation=<rot 90 0 0>}]
 dyn_hair_front_a_02_m [{shapeType=Capsule, length=2, radius=1, shapeName=dyn_hair_front_a_02_m_capsule, center=<3 2 0>, rotation=<rot 90 0 0>}]
 dyn_hat_bk01 [{
   shapeType=Capsule
   length=32.4629
   radius=5.7859
   shapeName=dyn_hat_bk01_capsule
   center=<8.072 -0.2931 9.737>
   rotation=<rot -0.0002 2.1614 -85.3815>
  }]
 head [{shapeType=Capsule, length=5, radius=8, shapeName=head_capsule, center=<5.9722 4.353 -1.1164>}]
 neck_01 [{shapeType=Capsule, length=0.505, radius=0.505, shapeName=neck_01_capsule, rotation=<rot 0 -90 0>}]
 neck_02 [{shapeType=Capsule, length=0.505, radius=0.505, shapeName=neck_02_capsule, rotation=<rot 0 -90 0>}]
 pelvis [{shapeType=Capsule, length=28.255, radius=16.13, shapeName=pelvis_capsule, center=<-9.4768 -3.4394 0>, rotation=<rot 0 -90 0>}]
 spine_01 [{shapeType=Capsule, length=25, radius=12.88, shapeName=spine_03_capsule, center=<0.7031 -0.1204 0>, rotation=<rot 0 -90 0>}]
 spine_02 [{shapeType=Capsule, length=15, radius=10, shapeName=spine_02_capsule}]
 spine_03 [{shapeType=Capsule, length=17.505, radius=9.63, shapeName=spine_03_capsule, center=<-1.1817 2.2875 0>, rotation=<rot 0 -90 0>}]
 spine_04 [{shapeType=Capsule, length=15, radius=10, shapeName=spine_04_capsule}]
 spine_05 [{shapeType=Capsule, length=19.38, radius=8.005, shapeName=spine_05_capsule, center=<2.6216 0.6486 0>, rotation=<rot 0 -90 0>}]
 upperarm_l [{shapeType=Capsule, length=20, radius=6, shapeName=upperarm_l_capsule, center=<13.5111 0 0>, rotation=<rot 90 0 0>}]
 upperarm_r [{shapeType=Capsule, length=20, radius=6, shapeName=upperarm_r_capsule, center=<-13.5111 0 0>, rotation=<rot 90 180 0>}]
}
constraints [
 {
  bone1Name=dyn_Hair_Back_03
  bone2Name=dyn_Hair_Back_02
  swing1LimitDegrees=20
  swing2LimitDegrees=35
  twistLimitDegrees=45
  twistMotion=Locked
 }
 {
  bone1Name=dyn_Hair_Back_04
  bone2Name=dyn_Hair_Back_03
  swing1LimitDegrees=15
  swing2LimitDegrees=30
  twistLimitDegrees=45
  twistMotion=Locked
 }
 {
  bone1Name=dyn_Hair_Back_06
  bone2Name=dyn_Hair_Back_04
  swing1LimitDegrees=15
  swing2LimitDegrees=30
  twistLimitDegrees=45
  twistMotion=Locked
 }
 {bone1Name=dyn_Hair_Back_02, bone2Name=head, swing1LimitDegrees=25, swing2LimitDegrees=40, twistLimitDegrees=45, twistMotion=Locked}
 {
  bone1Name=spine_05
  bone2Name=spine_04
  swing1LimitDegrees=45
  swing1Motion=Free
  swing2LimitDegrees=45
  swing2Motion=Free
  twistLimitDegrees=45
  twistMotion=Free
 }
 {
  bone1Name=spine_03
  bone2Name=spine_02
  swing1LimitDegrees=45
  swing1Motion=Free
  swing2LimitDegrees=45
  swing2Motion=Free
  twistLimitDegrees=45
  twistMotion=Free
 }
 {
  bone1Name=dyn_hat_bk01
  bone2Name=head
  swing1LimitDegrees=45
  swing1Motion=Locked
  swing2LimitDegrees=45
  swing2Motion=Locked
  twistLimitDegrees=45
  twistMotion=Locked
 }
 {
  bone1Name=spine_01
  bone2Name=pelvis
  swing1LimitDegrees=45
  swing1Motion=Free
  swing2LimitDegrees=45
  swing2Motion=Free
  twistLimitDegrees=45
  twistMotion=Free
 }
 {
  bone1Name=upperarm_l
  bone2Name=spine_05
  swing1LimitDegrees=45
  swing1Motion=Free
  swing2LimitDegrees=45
  swing2Motion=Free
  twistLimitDegrees=45
  twistMotion=Free
 }
 {
  bone1Name=upperarm_r
  bone2Name=spine_05
  swing1LimitDegrees=45
  swing1Motion=Free
  swing2LimitDegrees=45
  swing2Motion=Free
  twistLimitDegrees=45
  twistMotion=Free
 }
 {bone1Name=dyn_hair_front_a_01_m, bone2Name=head, swing1LimitDegrees=5, swing2LimitDegrees=5, twistLimitDegrees=0}
 {bone1Name=dyn_hair_front_a_02_m, bone2Name=dyn_hair_front_a_01_m, swing1LimitDegrees=5, swing2LimitDegrees=5, twistLimitDegrees=0}
 {
  bone1Name=FACIAL_L_ForeheadIn
  bone2Name=head
  swing1LimitDegrees=45
  swing1Motion=Locked
  swing2LimitDegrees=45
  swing2Motion=Locked
  twistLimitDegrees=45
  twistMotion=Locked
 }
 {
  bone1Name=FACIAL_L_ForeheadMid
  bone2Name=head
  swing1LimitDegrees=45
  swing1Motion=Locked
  swing2LimitDegrees=45
  swing2Motion=Locked
  twistLimitDegrees=45
  twistMotion=Locked
 }
 {
  bone1Name=head
  bone2Name=neck_02
  swing1LimitDegrees=45
  swing1Motion=Free
  swing2LimitDegrees=45
  swing2Motion=Free
  twistLimitDegrees=45
  twistMotion=Free
 }
 {
  bone1Name=neck_02
  bone2Name=neck_01
  swing1LimitDegrees=45
  swing1Motion=Free
  swing2LimitDegrees=45
  swing2Motion=Free
  twistLimitDegrees=45
  twistMotion=Free
 }
 {
  bone1Name=neck_01
  bone2Name=spine_05
  swing1LimitDegrees=45
  swing1Motion=Free
  swing2LimitDegrees=45
  swing2Motion=Free
  twistLimitDegrees=45
  twistMotion=Free
 }
 {
  bone1Name=spine_04
  bone2Name=spine_03
  swing1LimitDegrees=45
  swing1Motion=Free
  swing2LimitDegrees=45
  swing2Motion=Free
  twistLimitDegrees=45
  twistMotion=Free
 }
 {
  bone1Name=spine_02
  bone2Name=spine_01
  swing1LimitDegrees=45
  swing1Motion=Free
  swing2LimitDegrees=45
  swing2Motion=Free
  twistLimitDegrees=45
  twistMotion=Free
 }
 {
  bone1Name=spine_04
  bone2Name=spine_02
  swing1LimitDegrees=45
  swing1Motion=Free
  swing2LimitDegrees=45
  swing2Motion=Free
  twistLimitDegrees=45
  twistMotion=Free
 }
 {
  bone1Name=spine_03
  bone2Name=spine_01
  swing1LimitDegrees=45
  swing1Motion=Free
  swing2LimitDegrees=45
  swing2Motion=Free
  twistLimitDegrees=45
  twistMotion=Free
 }
 {
  bone1Name=spine_05
  bone2Name=spine_03
  swing1LimitDegrees=45
  swing1Motion=Free
  swing2LimitDegrees=45
  swing2Motion=Free
  twistLimitDegrees=45
  twistMotion=Free
 }
]
